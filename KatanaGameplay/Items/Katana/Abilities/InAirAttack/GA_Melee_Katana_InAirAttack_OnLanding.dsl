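; Katana in-air attack gameplay ability that holds back the slam technique until the player lands.
; Its non-authority activation adds a block-slam tag, optionally pops the pawn upward while preserving its fall start height, and optionally calls end on sliding.
; Ending removes the tag, then ends the Mantis technique if swimming or otherwise simulates a Primary release once to force the slam.
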(event Ability|EventActivateAbility
  (Variables|Default|SetbHasSimulatedInputEvent false)
  (Ability|HasAuthority
    (:True
      _)
    (:False
      (bind _outfortplayerpawn (CallFunction|GetAvatarasValidFortPlayerPawn)
        (:ValidFortPlayerPawn
          (Ability|GameplayEffect|AddLooseGameplayTags _outfortplayerpawn (Variables|Default|GetBlockSlamTag)))
        (:AvatarCastFailed
          _))
      (|CallParentFunction)
      (Utilities|FlowControl|Sequence
        (:then_0
          (if (ScalableFloat|GetValueasBool (Variables|Default|GetLaunchCharacterOnInAirAttack))
            (bind _outfortplayerpawn_2 (CallFunction|GetAvatarasValidFortPlayerPawn)
              (:ValidFortPlayerPawn
                (if (not (GameplayTags|HasAnyMatchingGameplayTags _outfortplayerpawn_2 (Variables|Default|GetTC_IgnoreApplication)))
                  (bind _asfort_movement_comp_character (Utilities|Casting|CastToFortMovementComp_Character (Variables|/Script/Engine.Character|GetCharacterMovement _outfortplayerpawn_2)))
                  (Variables|Default|SetFallStartHeightBeforeLaunch (Class|FortMovementCompCharacter|GetFallingStartedZ _asfort_movement_comp_character))
                  ;[1 1 1] Pop player up slightly
                  (Character|LaunchCharacter _outfortplayerpawn_2 (Math|Vector|MakeVector 0.0 0.0 (ScalableFloat|GetValueatLevel (Variables|Default|GetLaunchZMagnitude))) false true)
                  (Pawn|Character|SetFallingHeight _outfortplayerpawn_2 (Variables|Default|GetFallStartHeightBeforeLaunch)))))))
        (:then_1
          (if (ScalableFloat|GetValueasBool (Variables|Default|GetStopSlamOnSlide))
            (bind _asynctaskproxy (Ability|Tasks|WaitGameplayTagAdd "(TagName=\"MovementMode.Sliding\")" 0 true)
              (:then
                (Variables|Default|SetWaitSlidingAsyncTask _asynctaskproxy))
              (:Added
                (CallFunction|CallEnd)))))))))

(event Ability|EventFailedToActivatePassiveAbility
  (params (ActorInfo :type Gameplay%20Ability%20Actor%20Info%20Structure))
  (|CallParentFunction ActorInfo))

(event EventCallEnd
  (CallFunction|TryStartNextTechnique)
  (|CallParentFunction))

(event Ability|EventOnEndAbility
  (params (bWasCancelled :type Boolean))
  (|CallParentFunction bWasCancelled)
  (CallFunction|TryStartNextTechnique)
  (bind _waitslidingasynctask (Variables|Default|GetWaitSlidingAsyncTask)
    (:then
      (GameplayTasks|EndTask _waitslidingasynctask))
    (:else)))

(event Custom|TryStartNextTechnique
  ;[1 1 1] Force Slam Technique to play when landing on the ground.
  (bind _outfortplayerpawn (CallFunction|GetAvatarasValidFortPlayerPawn)
    (:ValidFortPlayerPawn
      (Ability|GameplayEffect|RemoveLooseGameplayTags _outfortplayerpawn (Variables|Default|GetBlockSlamTag))
      (if (GameplayTags|HasAnyMatchingGameplayTags _outfortplayerpawn (Variables|Default|GetSwimmingTag))
        (Utilities|IsValid (Actor|GetComponentbyClass _outfortplayerpawn "/Script/MantisRuntime.FortMantisPawnComponent")
          (:Is%20Valid
            (Mantis|TryManuallyEndCurrentTechnique (Actor|GetComponentbyClass _outfortplayerpawn "/Script/MantisRuntime.FortMantisPawnComponent"))))
        (elif (Variables|Default|GetbHasSimulatedInputEvent)
          _
          (else
            (Variables|Default|SetbHasSimulatedInputEvent true)
            (Utilities|IsValid (Actor|GetComponentbyClass _outfortplayerpawn "/Script/MantisRuntime.FortMantisPawnComponent")
              (:Is%20Valid
                (Mantis|SimulateInputEvent (Actor|GetComponentbyClass _outfortplayerpawn "/Script/MantisRuntime.FortMantisPawnComponent") "Primary" "Released" true true)))))))))
